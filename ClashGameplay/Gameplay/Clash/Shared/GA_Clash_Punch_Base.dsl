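; Base gameplay ability for a Clash punch that launches a projectile, with empty hit and cleanup events for child abilities.
; Init attaches idle hand FX; activation clears it and listens for projectile gameplay events.
; Spawn stores the projectile and applies an effect container to selected targets to destroy nearby buildings; hit jumps the montage to "Pull" and records success.
; A montage notify ends the ability early after successful hits, on non-locally-controlled instances only.
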(event Custom|SpawnIdleFX
  ;[1 0.1131 0.9482] Idle FX Stuff
  ;;[1 1 1] Event->SpawnIdleFX.  This is being called from the InitAbility function - spawns FXs at pawns hand.
  (bind _returnvalue (Niagara|SpawnSystemAttached (Variables|Default|GetIdleFXSystem) (Variables|/Script/Engine.Character|GetMesh (Variables|Default|GetPlayerPawn)) "hand_r" "0, 0, 0" "0, 0, 0" "KeepRelativeOffset" true))
  (bind _output_get (Variables|Default|SetFX_Idle _returnvalue))
  (Effects|Components|ParticleSystem|SetBoolParameter _output_get (Variables|Default|GetFXColorParamName) (Variables|Default|GetFXIsRed)))

(event Ability|EventActivateAbility
  ;[1 1 1] EventReleased-> Block player input and start punch
  (|CallParentFunction)
  (CallFunction|ReleasedCopy)
  ;[1 1 1] Plays return section of montage when the projectile hits an eligible target
  (bind (_asynctaskproxy _payload) (Ability|Tasks|WaitGameplayEvent "(TagName=\"Event.Clash.Punch.Projectile.Hit\")")
    (:then
      (bind (_asynctaskproxy_2 _payload_2) (Ability|Tasks|WaitGameplayEvent "(TagName=\"Event.Clash.Punch.Projectile.Spawn\")")
        (:EventReceived
          (CallFunction|DestroyBuildings)
          (bind _asactor (Utilities|Casting|CastToActor (.OptionalObject (Utilities|Struct|BreakGameplayEventData _payload_2))))
          (Variables|Default|SetSpawnedProjectile _asactor))))
    (:EventReceived
      (Ability|Animation|MontageJumptoSection "Pull")
      (Variables|Default|SetHitATarget (Math|Conversions|ToBoolean(Integer) (Math|Float|Truncate (.EventMagnitude (Utilities|Struct|BreakGameplayEventData _payload))))) ; Projectile passes whether we hit a target or not
      (bind _asactor_2 (Utilities|Casting|CastToActor (.OptionalObject (Utilities|Struct|BreakGameplayEventData _payload))))
      (Development|PrintString (Utilities|GetDisplayName _asactor_2))
      (CallFunction|OnPunchHit _asactor_2))))

(event Custom|Released_Copy
  (bind _fx_idle (Variables|Default|GetFX_Idle))
  ;[1 0.1131 0.9482] Idle FX Stuff
  ;;[1 1 1] Event-> Released Copy.  Clean up idle FX.
  (Utilities|IsValid _fx_idle
    (:Is%20Valid
      (Components|Activation|Deactivate _fx_idle))))

(event Custom|DestroyBuildings
  ;[0.9634 0.9469 1] Destroy Nearby buildings when ability fires
  (bind (_asynctaskproxy _targetdata _applicationtag) (Ability|Tasks|DoTargetSelection "None" "(TagName=\"Targeting.ApplicationTag.1\")" true)
    (:Targeted
      (Ability|ApplyGameplayEffectContainer _targetdata _applicationtag))))

(event Custom|OnPunchHit
  (params (HitActor :type Actor%20Object%20Reference%20%28by%20ref%29))
  ;[1 1 1] Don't delete, used by children
  (Utilities|IsValid HitActor) ; Trivial node so this event isn't flagged as empty
)

(event Ability|EventOnEndAbility
  (params (bWasCancelled :type Boolean))
  ;[1 1 1] Reset Vars and call clean up event
  (Development|PrintString "ability ended")
  (Variables|Default|SetHitATarget false) ; Projectile passes whether we hit a target or not
  (CallFunction|CleanupActors)
  (|CallParentFunction bWasCancelled))

(event EventCastMontageNotifyStart
  (params (NotifyTag :type Gameplay%20Tag%20Structure))
  ;[1 1 1] End ability early if we hit a target.  Allows pawn to follow up.
  (if (GameplayTags|Equal(GameplayTag) (Variables|Default|GetT%20SuccesfulHitAbilityEnd) NotifyTag)
    (Ability|IsLocallyControlled
      (:False
        (if (Variables|Default|GetHitATarget)
          (Utilities|FlowControl|Delay 0.25)
          (Ability|EndAbility))))
    (else
      (|CallParentFunction NotifyTag))))

(event Custom|CleanupActors)

(event Custom|Handle%20Look/Rot
  (params (Disable :type Boolean))
  (bind _playerpawn (Variables|Default|GetPlayerPawn))
  ;[1 0.07292 0.07292] Unused
  (AI|Movement|StopActiveMovement (Variables|/Script/Engine.Character|GetCharacterMovement _playerpawn))
  (CallFunction|SetBlockMovement _playerpawn Disable)
  (CallFunction|ManageLookInput Disable))

(fn InitAbility
  (|CallParentFunction)
  (CallFunction|SpawnIdleFX))

(fn ManageLookInput
  (params (LockLook :type Boolean))
  (Input|SetIgnoreLookInput (Fortnite|GetFortPlayerControllerfromActor (Variables|Default|GetPlayerPawn)) LockLook))

(fn OnRep_HitATarget
  (if (not (Variables|Default|GetHitATarget))
    (Variables|Default|SetHitATarget true)))

(fn OnRep_SpawnedPrj)
