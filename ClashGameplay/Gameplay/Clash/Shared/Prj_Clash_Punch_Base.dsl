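; Base projectile for the Clash punch ability: a homing hand that can hit and yank an enemy.
; It homes on the aimed enemy, turns back to its owner after a timer, or attaches to a valid target and signals the ability via gameplay events, with a failsafe for target death.
; A repeating server capsule trace damages buildings ahead of the projectile or yanked actor.
; Timelines drive material and Niagara parameters for the stretching red or default arm.
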(event EventBeginPlay
  (bind _staticmesh (Variables|Default|GetStaticMesh))
  (bind _returnvalue (Transformation|GetWorldLocation _staticmesh))
  (bind _returnvalue_2 (Projectile|GetCurrentLocation :!class /Script/FortniteGame.FortProjectileMovementInterface))
  (bind _fortpawn (Variables|Default|GetFortPawn))
  (bind _mesh (Variables|/Script/Engine.Character|GetMesh _fortpawn))
  (bind _ns_hightower_vertigo_handmesh_fx (Variables|Default|GetNS_HighTower_Vertigo_HandMesh_FX))
  (bind _staticmesh_2 (Variables|Default|GetStaticMesh))
  ;[1 1 1] BeginPlay-> Set homing vars and projectile return timer.  Send event to attach FX beam to punch ability
  (|CallParentFunction)
  (if (Variables|Default|GetFXIsRed)
    (Rendering|Material|SetMaterial (Variables|Default|GetSM_Hightower_Vertigo_HandAnimation_HandA) 0 "/ClashGameplay/Gameplay/Clash/FX/MI_Clash_Big_Hand_Mesh.MI_Clash_Big_Hand_Mesh")
    (Components|StaticMesh|SetStaticMesh (Variables|Default|GetStaticMesh) "/ClashGameplay/Gameplay/Clash/FX/SM_Clash_Arm_Trail_Red.SM_Clash_Arm_Trail_Red")
    (Rendering|Material|SetMaterial (Variables|Default|GetStaticMesh) 0 "/ClashGameplay/Gameplay/Clash/FX/MI_Clash_Arm.MI_Clash_Arm")
    (else
      (Rendering|Material|SetMaterial (Variables|Default|GetSM_Hightower_Vertigo_HandAnimation_HandA) 0 "/ClashGameplay/Gameplay/Vertigo/FX/M_HighTower_Vertigo_Hand_Inst.M_HighTower_Vertigo_Hand_Inst")
      (Components|StaticMesh|SetStaticMesh _staticmesh_2 "/ClashGameplay/Gameplay/Vertigo/FX/SM_Hightower_Vertigo_Arm.SM_Hightower_Vertigo_Arm")
      (Rendering|Material|SetMaterial _staticmesh_2 0 "/ClashGameplay/Gameplay/Vertigo/FX/M_HighTower_Vertigo_Arm_Inst.M_HighTower_Vertigo_Arm_Inst")))
  (Variables|Default|SetBullDozer self)
  (CallFunction|InitializeFX)
  (CallFunction|SetHomingVars false)
  (Utilities|Time|SetTimerbyEvent (delegate Custom|TriggerReturn) 0.6)
  (bind _returnvalue_3 (Utilities|Time|SetTimerbyEvent (EventDispatchers|CreateEvent :!event Custom|BulidingDestructionTrace) 0.1 true))
  (Variables|Default|SetBuildingDestroyTimer _returnvalue_3)
  (Ability|SendGameplayEventtoActor (Variables|Default|GetFortPawn) "(TagName=\"Event.Clash.Punch.Projectile.Spawn\")" (Utilities|Struct|MakeGameplayEventData :OptionalObject self :ContextHandle "()" :InstigatorTags "(GameplayTags=)" :TargetTags "(GameplayTags=)" :EventMagnitude 0.0 :TargetData "()")) ; Sends event to fire beam logic to the punch ability
  (SoundIndicator|ShowSoundIndicator (Variables|Default|GetFortSoundIndicator) (Transformation|GetActorLocation) "/ClashGameplay/Sounds/Clash/Clash_Red/Clash_Red_SNG_Cast.Clash_Red_SNG_Cast" 1.0 0)
  ;[1 1 1] FX_Arm Part
  (Components|Activation|Activate _ns_hightower_vertigo_handmesh_fx true)
  (Utilities|FlowControl|Sequence
    (:then_0
      (bind _returnvalue_4 (Niagara|SpawnSystemAttached (Variables|Default|GetNS_HandFX) _mesh "hand_r" "0, 0, 0" "0, 0, 0" "KeepRelativeOffset" true true "None" false))
      (bind _output_get (Variables|Default|SetHandParticle _returnvalue_4))
      (Effects|Components|ParticleSystem|SetBoolParameter _output_get (Variables|Default|GetFXColorParamName) (Variables|Default|GetFXIsRed))
      (Transformation|AttachComponentToComponent (Variables|Default|GetStaticMesh) _mesh "hand_r" "KeepRelative" "SnapToTarget" "KeepRelative" true)
      (bind (_direction _sinlength _sinmotion _sinintensity _disslove _disslove_hand _tentacles_dissolve) (Timeline|Timeline_NotHit PlayFromStart)
        (:Update
          (Rendering|Material|SetVectorParameterValueonMaterials _staticmesh "TargetPosition" _returnvalue_2)
          (Rendering|Material|SetScalarParameterValueonMaterials _staticmesh "SinMotion" _sinmotion)
          (Rendering|Material|SetScalarParameterValueonMaterials (Variables|Default|GetSM_Hightower_Vertigo_HandAnimation_HandA) "Dissolve" _disslove_hand)
          (if (Variables|Default|GetHit)
            (Effects|Components|ParticleSystem|SetFloatParameter (Variables|Default|GetNS_HighTower_Vertigo_HandMesh_FX) "Tentacles_Dissolve" _tentacles_dissolve))
          (Rendering|Material|SetScalarParameterValueonMaterials _staticmesh "SinLength" _sinlength)
          (Rendering|Material|SetScalarParameterValueonMaterials _staticmesh "SinIntensity" (* _sinintensity 5e+01))
          (Rendering|Material|SetScalarParameterValueonMaterials _staticmesh "Dissolve" _disslove)
          (Transformation|SetWorldRotation _staticmesh (Math|Rotator|FindLookatRotation _returnvalue _returnvalue_2) false false)
          (Effects|Components|ParticleSystem|SetVectorParameter (Variables|Default|GetHandParticle) "TargetPosition" _returnvalue_2)
          (Rendering|SetBoundsScale _staticmesh (Math|Vector|VectorLength (- _returnvalue _returnvalue_2))))))
    (:then_1
      (Utilities|FlowControl|Delay 0.5)
      (Components|Activation|Deactivate _ns_hightower_vertigo_handmesh_fx))))

(event Custom|TriggerReturn
  ;[1 1 1] BeginPlay-> Set homing vars and projectile return timer.  Send event to attach FX beam to punch ability
  (Variables|Default|SetReturnTimerExpired true))

(event Custom|Hit_Copy
  (bind _staticmesh (Variables|Default|GetStaticMesh))
  (bind _returnvalue (Transformation|GetWorldLocation _staticmesh))
  (bind _returnvalue_2 (Projectile|GetCurrentLocation :!class /Script/FortniteGame.FortProjectileMovementInterface))
  (bind _staticmesh_2 (Variables|Default|GetStaticMesh))
  (bind _staticmesh_3 (Variables|Default|GetStaticMesh))
  (bind (_hitinfo_bblockinghit _hitinfo_binitialoverlap _hitinfo_time _hitinfo_distance _hitinfo_location _hitinfo_impactpoint _hitinfo_normal _hitinfo_impactnormal _hitinfo_physmat _hitinfo_hitactor) (Collision|BreakHitResult (Variables|GetHitInfo_bBlockingHit)))
  (bind _sm_hightower_vertigo_handanimation_handa (Variables|Default|GetSM_Hightower_Vertigo_HandAnimation_HandA))
  (bind _ns_impact (Variables|Default|GetNS_Impact))
  (bind (_hitinfo_bblockinghit_2 _hitinfo_binitialoverlap_2 _hitinfo_time_2 _hitinfo_distance_2 _hitinfo_location_2) (Collision|BreakHitResult (Variables|GetHitInfo_bBlockingHit)))
  ;[1 1 1] FX_Arm Part
  (Networking|CanExecuteCosmeticEvents
    (:True
      (Utilities|FlowControl|Sequence
        (:then_0
          (Rendering|SetVisibility _sm_hightower_vertigo_handanimation_handa false false)
          (Components|DestroyComponent _sm_hightower_vertigo_handanimation_handa)
          (Components|Activation|Activate (Variables|Default|GetNS_HighTower_Vertigo_HandAnimation) false)
          (bind _asplayer_pawn_athena (Utilities|Casting|CastToPlayerPawn_Athena _hitinfo_hitactor)
            (:then
              (bind _returnvalue_3 (Effects|SpawnFXSystemAttached _ns_impact (Variables|/Script/Engine.Character|GetMesh _asplayer_pawn_athena) "spine_01"))
              (Effects|Components|ParticleSystem|SetBoolParameter _returnvalue_3 (Variables|Default|GetFXColorParamName) (Variables|Default|GetFXIsRed))
              (Rendering|Material|SetScalarParameterValueonMaterials _staticmesh_3 "CosineOffset" (Math|Random|RandomFloat))
              (bind (_direction _sinlength _sinintensity _dissolve) (Timeline|Timeline_Hit PlayFromStart)
                (:Update
                  (Rendering|Material|SetScalarParameterValueonMaterials _staticmesh_2 "SinLength" _sinlength)
                  (Rendering|Material|SetScalarParameterValueonMaterials _staticmesh_2 "SinIntensity" (* _sinintensity (Math|Float|Lerp 0.2 1.0 (Math|Float|Clamp(Float) (/ (Math|Vector|VectorLength (- (Transformation|GetWorldLocation _staticmesh_3) _hitinfo_location_2)) 3.2e+03)))))
                  (Rendering|Material|SetScalarParameterValueonMaterials _staticmesh_2 "Dissolve" _dissolve)
                  (Transformation|SetWorldRotation _staticmesh_2 (Math|Rotator|FindLookatRotation _returnvalue _returnvalue_2) false false)
                  (Effects|Components|ParticleSystem|SetVectorParameter (Variables|Default|GetHandParticle) "TargetPosition" _returnvalue_2)
                  (Rendering|SetBoundsScale _staticmesh (Math|Vector|VectorLength (- _returnvalue _returnvalue_2))))))
            (:CastFailed
              (bind _returnvalue_4 (Niagara|SpawnSystematLocation _ns_impact _hitinfo_impactpoint))
              (Effects|Components|ParticleSystem|SetBoolParameter _returnvalue_4 (Variables|Default|GetFXColorParamName) (Variables|Default|GetFXIsRed)))))))))

(event Custom|BulidingDestructionTrace
  (bind _bulldozer (Variables|Default|GetBullDozer))
  ;[1 1 1] Trace 6 frames infront of our proj or yanked actor to destroy buildings.  The punch ability handles very close buildings.
  (if (Networking|IsServer)
    (Utilities|IsValid _bulldozer
      (:Is%20Valid
        (bind _outhits (Collision|MultiCapsuleTraceForObjects (Transformation|GetActorLocation _bulldozer) (+ (Transformation|GetActorLocation _bulldozer) (* (Transformation|GetVelocity _bulldozer) (* (Utilities|Time|GetWorldDeltaSeconds) 6.0))) 4e+01 4e+01 (Variables|Default|GetWhileLeapingTargetObjectTypes) false (Variables|Default|GetActorsToIgnore)))
        (CallFunction|BuildingDestruction _outhits)))))

(event EventEndPlay
  (params (EndPlayReason :type EEndPlayReason%20Enum :default "Destroyed"))
  (bind _handparticle (Variables|Default|GetHandParticle))
  (Components|Activation|Deactivate _handparticle)
  (Utilities|FlowControl|Delay 0.2)
  (Components|DestroyComponent _handparticle))

(event Collision|EventHit
  (params (MyComp :type Primitive%20Component%20Object%20Reference) (Other :type Actor%20Object%20Reference) (OtherComp :type Primitive%20Component%20Object%20Reference) (bSelfMoved :type Boolean) (HitLocation :type Vector :default "0, 0, 0") (HitNormal :type Vector :default "0, 0, 0") (NormalImpulse :type Vector :default "0, 0, 0") (Hit :type Hit%20Result%20Structure%20%28by%20ref%29))
  (bind (_asbuilding_actor _bsuccess) (Utilities|Casting|CastToBuildingActor Other))
  ;[1 1 1] EventHit-> If we hit a valid target, attach and start yank.  Make that actor the thing that destroys buildings.
  (Variables|Default|SetHit true)
  (Variables|Default|SetHitInfo Hit)
  (CallFunction|HitCopy)
  (Collision|SetActorEnableCollision false)
  (if (Math|Boolean|ANDBoolean (Variables|/Script/Engine.Actor|GetbCanBeDamaged Other) (not _bsuccess) (not (Team|OnSameTeam (Variables|Default|GetFortPawn) Other)))
    (Variables|Default|SetHitValidTarget true)
    (Variables|Default|SetBullDozer Other)
    (Variables|/Script/Engine.ProjectileMovementComponent|SetbRotationFollowsVelocity false (Variables|Movement|GetProjectileMovementComponent))
    (Ability|SendGameplayEventtoActor (Variables|Default|GetFortPawn) "(TagName=\"Event.Clash.Punch.Projectile.Hit\")" (Utilities|Struct|MakeGameplayEventData :OptionalObject Other :ContextHandle "()" :InstigatorTags "(GameplayTags=)" :TargetTags "(GameplayTags=)" :EventMagnitude 0.0 :TargetData "()")) ; Sends event to change montage section to yank ability
    (bind _asfort_player_pawn (Utilities|Casting|CastToFortPlayerPawn (Variables|Default|GetBullDozer))
      (:then
        (Tags|AddNonReplicatedTagtoPawn _asfort_player_pawn (Variables|Default|GetT_CancelAbility))
        (Utilities|Time|SetTimerbyEvent (delegate Custom|CustomEvent_1) 0.3))
      (:CastFailed
        _))
    (Transformation|AttachActorToActor Other "" "KeepRelative" "KeepWorld" "KeepWorld")))

(event Custom|CustomEvent_1
  ;[1 1 1] Fixes a bug that was causing looping montages to not cancel on the owning client since they don't get the server call to end it
  (Tags|RemoveNonReplicatedTagtoPawn (stale Collision|EventHit _asfort_player_pawn) (Variables|Default|GetT_CancelAbility)))

(event ProjectileMovement|EventOnReachedLockTargetDistanceThreshold
  ;[1 1 1] Kill prj when it returns without hitting anything
  (if (Variables|Default|GetReturnTimerExpired)
    (Actor|DestroyActor)))

(event EventOnAttachedToDied
  (params (DamagedActor :type Actor%20Object%20Reference) (Damage :type Float%20%28single-precision%29) (InstigatedBy :type Controller%20Object%20Reference) (DamageCauser :type Actor%20Object%20Reference) (HitLocation :type Vector :default "0, 0, 0") (FHitComponent :type Primitive%20Component%20Object%20Reference) (BoneName :type Name :default "None") (Momentum :type Vector :default "0, 0, 0"))
  ;[1 1 1] EventAttachedTargetDied-> Failsafe for if the pulled target dies.  This will make the punch return flow exactly the same even if the grab ability is not called.  Fixes a bug where the pull ability flow would not work if target was killed.
  (Utilities|Time|SetTimerbyEvent (delegate Custom|CustomEvent_0) 0.5))

(event Custom|CustomEvent_0
  (bind (_bblockinghit _binitialoverlap _time _distance _location _impactpoint _normal _impactnormal _physmat _hitactor) (Collision|BreakHitResult (Variables|Default|GetHitInfo)))
  ;[1 1 1] EventAttachedTargetDied-> Failsafe for if the pulled target dies.  This will make the punch return flow exactly the same even if the grab ability is not called.  Fixes a bug where the pull ability flow would not work if target was killed.
  (Ability|SendGameplayEventtoActor (Variables|Default|GetFortPawn) "(TagName=\"Event.Clash.Punch.Projectile.Hit\")" (Utilities|Struct|MakeGameplayEventData :OptionalObject _hitactor :ContextHandle "()" :InstigatorTags "(GameplayTags=)" :TargetTags "(GameplayTags=)" :EventMagnitude (Math|Conversions|ToFloat(Boolean) true) :TargetData "()")) ; Sends event to change montage section to yank ability
)

(fn BuildingDestruction
  (params (Array :type Array%20of%20Hit%20Result%20Structures))
  (for _array_element Array
    (Utilities|IsValid (.HitActor (Collision|BreakHitResult _array_element))
      (:Is%20Valid
        (if (and (<= (Math|Vector|VectorLength (- (Transformation|GetActorLocation (.HitActor (Collision|BreakHitResult _array_element))) (Variables|Default|GetStartingLocation))) 5e+03) (> (Math|Vector|DotProduct (Math|Vector|Normalize (- (Transformation|GetActorLocation (.HitActor (Collision|BreakHitResult _array_element))) (Transformation|GetActorLocation (Variables|Default|GetFortPawn)))) (Math|Vector|Normalize (Transformation|GetActorForwardVector (Variables|Default|GetFortPawn)))) 0.0))
          (bind _asbuilding_actor (Utilities|Casting|CastToBuildingActor (.HitActor (Collision|BreakHitResult _array_element))))
          (if (GameplayTags|HasAnyMatchingGameplayTags _asbuilding_actor (GameplayTags|MakeGameplayTagContainerfromArray (Utilities|Array|MakeArray "(TagName=\"Trait.Reinforced\")")))
            (if (Variables|Default|GetHitValidTarget)
              _))
          (if (Variables|/Script/Engine.Actor|GetbCanBeDamaged _asbuilding_actor)
            (GameplayEffects|ApplyGameplayEffectToTarget :self (Ability|GetFortAbilitySystemComponent (Variables|Default|GetFortPawn)) :GameplayEffectClass (Variables|Default|GetGE_PunchDamage) :Target (Ability|GetFortAbilitySystemComponent _asbuilding_actor) :Level 0.0)))))))

(fn GetAffiliationsToShowFor
  (return "Friendly"))

(fn InitializeFX
  (Effects|Components|ParticleSystem|SetBoolParameter (Variables|Default|GetNS_HighTower_Vertigo_HandAnimation) (Variables|Default|GetFXColorParamName) (Variables|Default|GetFXIsRed))
  (Effects|Components|ParticleSystem|SetBoolParameter (Variables|Default|GetNS_HighTower_Vertigo_HandMesh_FX) (Variables|Default|GetFXColorParamName) (Variables|Default|GetFXIsRed)))

(fn OnRep_ReturnTimerExpired
  (if (Variables|Default|GetReturnTimerExpired)
    (if (not (Variables|Default|GetHitValidTarget))
      (Collision|SetActorEnableCollision false)
      (Variables|/Script/Engine.ProjectileMovementComponent|SetbRotationFollowsVelocity false (Variables|Movement|GetProjectileMovementComponent))
      (CallFunction|SetHomingVars true))))

(fn SetHomingVars
  (params (Return? :type Boolean))
  (bind _self self)
  (bind _homingtarget (Variables|Default|GetHomingTarget))
  (if Return?
    (Variables|Default|SetHomingTarget (Variables|Default|GetFortPawn))
    (else
      (bind _returnvalue (Fortnite|GetAimedAtEnemy (Fortnite|GetFortPlayerControllerfromActor (Variables|Default|GetFortPawn)) (ScalableFloat|GetValueatLevel (Variables|HotFixVars|GetDataHomingRectWidth)) (ScalableFloat|GetValueatLevel (Variables|HotFixVars|GetDataHomingRectHeight)) (ScalableFloat|GetValueatLevel (Variables|HotFixVars|GetDataHomingRange)) true))
      (Utilities|IsValid _returnvalue
        (:Is%20Valid
          (if (not (GameplayTags|HasMatchingGameplayTag _returnvalue (Variables|Default|GetT_DamageImmune)))
            (Variables|Default|SetHomingTarget _returnvalue))))))
  (Projectile|SetHomingStyle 0 0 "LockOn" false :!class /Script/FortniteGame.FortProjectileMovementInterface)
  (Networking|SwitchHasAuthority
    (:Authority
      (Projectile|SetHomingTurnSpeedMin 0 _self (ScalableFloat|GetValueatLevel (Variables|HotFixVars|GetDataHomingTurnSpeedMin)) :!class /Script/FortniteGame.FortProjectileMovementInterface)
      (Projectile|SetHomingTurnSpeedMax 0 _self (ScalableFloat|GetValueatLevel (Variables|HotFixVars|GetDataHomingTurnSpeedMax)) :!class /Script/FortniteGame.FortProjectileMovementInterface)
      (Projectile|SetHomingTimeUntilMaxTurnSpeed 0 _self 0.0 :!class /Script/FortniteGame.FortProjectileMovementInterface)
      (Projectile|SetHomingTargetOffset _self (Variables|Default|GetTargetHomingOffset) :!class /Script/FortniteGame.FortProjectileMovementInterface)
      (Utilities|IsValid _homingtarget
        (:Is%20Valid
          (Projectile|SetHomingTarget 0 self _homingtarget 0 :!class /Script/FortniteGame.FortProjectileMovementInterface)
          (if Return?
            (Projectile|SetHomingTargetOffset self (Variables|Default|GetReturnHomingOffset) :!class /Script/FortniteGame.FortProjectileMovementInterface)
            (Projectile|SetHomingLockTargetDistanceThreshold 0 self 1e+02 :!class /Script/FortniteGame.FortProjectileMovementInterface)
            (Projectile|ResumeSimulation)))))))

(fn ShouldShowSoundIndicator
  (params (PlayerController :type Fort%20Player%20Controller%20Object%20Reference))
  (return (and (Utilities|NotEqual(Object) (Game|Player|GetPlayerPawn PlayerController) (Game|GetInstigator)) (or (Variables|/Script/FortniteGame.FortRuntimeOptions|GetbSoundIndicatorsEnabledForTeammates (Options|GetRuntimeOptions)) (Utilities|Enum|NotEqual(Enum) (Team|GetTeamAffiliation (Game|Player|GetPlayerPawn PlayerController) (Game|GetInstigator)) "Friendly")))))

(fn ConstructionScript
  (Variables|Default|SetStartingLocation (Transformation|GetActorLocation)))
